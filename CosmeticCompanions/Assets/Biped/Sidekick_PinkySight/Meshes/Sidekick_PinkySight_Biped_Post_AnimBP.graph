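# Post-process AnimBP for the Sidekick_PinkySight biped cosmetic companion that adds secondary motion to its four cloak panels.
# Input pose runs through rotation constraints tying cloak master bones to same-side thighs, then zero-offset additive bone modifiers.
# Root-relative AnimDynamics with angular springs, limits and gravity then simulates each panel's first bone.
# The template PoseAsset blend node is unconnected, and the sine-driven CloakRot1 from EventGraph is never read.
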AnimBlueprint /CosmeticCompanions/Assets/Biped/Sidekick_PinkySight/Meshes/Sidekick_PinkySight_Biped_Post_AnimBP
target_skeleton=&/Sidekick_PinkySight_Biped_Skeleton
graphs [
 AnimGraph {nodes [
   Pose!(PoseAsset:Pose Asset Object Reference=Get PoseAsset() -> (PoseAsset:Pose Asset Object Reference), SourcePose:Pose Link Structure) -> (Pose:Pose Link Structure) {is=title AnimGraphNode_PoseBlendNode, bIgnorePoseAssetSourceAnimationCurves=true, blendOption=Linear}
   n1 = "Constraint - dyn_cloak_bk_l_mstr (R)"(ComponentPose=Local To Component(LocalPose=Input Pose()), Alpha=1, ConstraintWeights_0=1) {
    boneToModify {boneName=dyn_cloak_bk_l_mstr}
    constraintSetup [{transformType=Rotation, targetBone {boneName=thigh_l}}]
    constraintWeights [1]
   }
   n2 = "Constraint - dyn_cloak_bk_r_mstr (R)"(ComponentPose=@n1, Alpha=1, ConstraintWeights_0=1) {
    boneToModify {boneName=dyn_cloak_bk_r_mstr}
    constraintSetup [{transformType=Rotation, targetBone {boneName=thigh_r}}]
    constraintWeights [1]
   }
   n3 = "Constraint - dyn_cloak_ft_l_mstr (R)"(ComponentPose=@n2, Alpha=1, ConstraintWeights_0=1) {
    boneToModify {boneName=dyn_cloak_ft_l_mstr}
    constraintSetup [{transformType=Rotation, targetBone {boneName=thigh_l}}]
    constraintWeights [1]
   }
   n4 = "Constraint - dyn_cloak_ft_r_mstr (R)"(ComponentPose=@n3, Alpha=1, ConstraintWeights_0=1) {
    boneToModify {boneName=dyn_cloak_ft_r_mstr}
    constraintSetup [{transformType=Rotation, targetBone {boneName=thigh_r}}]
    constraintWeights [1]
   }
   n5 = "Transform (Modify) Bone - Bone: dyn_cloak_bk_r_mstr"(ComponentPose=@n4, Alpha=0.25, Translation="0,0,0", Rotation="0,0,0", Scale="1,1,1") {
    rotationMode=BMM_Additive
    rotationSpace=BCS_BoneSpace
    translationMode=BMM_Additive
    translationSpace=BCS_BoneSpace
    boneToModify {boneName=dyn_cloak_bk_r_mstr}
   }
   n6 = "Transform (Modify) Bone - Bone: dyn_cloak_bk_l_mstr"(ComponentPose=@n5, Alpha=0.25, Translation="0,0,0", Rotation="0,0,0", Scale="1,1,1") {
    rotationMode=BMM_Additive
    rotationSpace=BCS_BoneSpace
    translationMode=BMM_Additive
    translationSpace=BCS_BoneSpace
    boneToModify {boneName=dyn_cloak_bk_l_mstr}
   }
   n7 = "Transform (Modify) Bone - Bone: dyn_cloak_ft_r_mstr"(ComponentPose=@n6, Alpha=0.25, Translation="0,0,0", Rotation="0,0,0", Scale="1,1,1") {
    rotationMode=BMM_Additive
    rotationSpace=BCS_BoneSpace
    translationMode=BMM_Additive
    translationSpace=BCS_BoneSpace
    boneToModify {boneName=dyn_cloak_ft_r_mstr}
   }
   n8 = "Transform (Modify) Bone - Bone: dyn_cloak_ft_l_mstr"(ComponentPose=@n7, Alpha=0.25, Translation="0,0,0", Rotation="0,0,0", Scale="1,1,1") {
    rotationMode=BMM_Additive
    rotationSpace=BCS_BoneSpace
    translationMode=BMM_Additive
    translationSpace=BCS_BoneSpace
    boneToModify {boneName=dyn_cloak_ft_l_mstr}
   }
   n9 = Anim Dynamics - Bone: dyn_cloak_ft_r_01(ComponentPose=@n8, Alpha=1, ExternalForce="0,0,0") {
    angularBiasOverride=1
    angularDampingOverride=1
    angularSpringConstant=2000
    bAngularSpring=true
    bOverrideAngularBias=true
    bOverrideAngularDamping=true
    bUseGravityOverride=true
    simulationSpace=RootRelative
    alphaBoolBlend {blendOption=Linear}
    alphaScaleBiasClamp {bias=1, scale=-1}
    boundBone {boneName=dyn_cloak_ft_r_01}
    componentAppliedLinearAccClamp=<3000 3000 3000>
    gravityOverride {z=-980}
    physicsBodyDefinitions [{
      boundBone {boneName=dyn_cloak_ft_r_01}
      boxExtents=<10 10 10>
      constraintSetup {angularLimitsMax=<0 5 0>, angularLimitsMin=<0 -5 -30>, angularTarget=<1 0 0>}
      localJointOffset=<-2 0 0>
     }]
    simSpaceSettings {maxAngularAcceleration=45, maxAngularVelocity=8, simSpaceAngularAlpha=0.25}
   }
   n10 = Anim Dynamics - Bone: dyn_cloak_ft_l_01(ComponentPose=@n9, Alpha=1, ExternalForce="0,0,0") {
    angularBiasOverride=1
    angularDampingOverride=1
    angularSpringConstant=2000
    bAngularSpring=true
    bOverrideAngularBias=true
    bOverrideAngularDamping=true
    bUseGravityOverride=true
    simulationSpace=RootRelative
    alphaBoolBlend {blendOption=Linear}
    alphaScaleBiasClamp {bias=1, scale=-1}
    boundBone {boneName=dyn_cloak_ft_l_01}
    componentAppliedLinearAccClamp=<3000 3000 3000>
    gravityOverride {z=-980}
    physicsBodyDefinitions [{
      boundBone {boneName=dyn_cloak_ft_l_01}
      boxExtents=<10 10 10>
      constraintSetup {angularLimitsMax=<0 5 30>, angularLimitsMin=<0 -5 0>, angularTarget=<1 0 0>}
      localJointOffset=<2 0 0>
     }]
    simSpaceSettings {maxAngularAcceleration=45, maxAngularVelocity=8, simSpaceAngularAlpha=0.25}
   }
   n11 = Anim Dynamics - Bone: dyn_cloak_bk_r_01(ComponentPose=@n10, Alpha=1, ExternalForce="0,0,0") {
    angularBiasOverride=1
    angularDampingOverride=1
    angularSpringConstant=2000
    bAngularSpring=true
    bOverrideAngularBias=true
    bOverrideAngularDamping=true
    bUseGravityOverride=true
    simulationSpace=RootRelative
    alphaBoolBlend {blendOption=Linear}
    alphaScaleBiasClamp {bias=1, scale=-1}
    boundBone {boneName=dyn_cloak_bk_r_01}
    componentAppliedLinearAccClamp=<3000 3000 3000>
    gravityOverride {z=-980}
    physicsBodyDefinitions [{
      boundBone {boneName=dyn_cloak_bk_r_01}
      boxExtents=<10 10 10>
      constraintSetup {angularLimitsMax=<0 5 0>, angularLimitsMin=<0 -5 -30>, angularTarget=<1 0 0>}
      localJointOffset=<-2 0 0>
     }]
    simSpaceSettings {maxAngularAcceleration=45, maxAngularVelocity=8, simSpaceAngularAlpha=0.25}
   }
   n12 = Anim Dynamics - Bone: dyn_cloak_bk_l_01(ComponentPose=@n11, Alpha=1, ExternalForce="0,0,0") {
    angularBiasOverride=1
    angularDampingOverride=1
    angularSpringConstant=2000
    bAngularSpring=true
    bOverrideAngularBias=true
    bOverrideAngularDamping=true
    bUseGravityOverride=true
    simulationSpace=RootRelative
    alphaBoolBlend {blendOption=Linear}
    alphaScaleBiasClamp {bias=1, scale=-1}
    boundBone {boneName=dyn_cloak_bk_l_01}
    componentAppliedLinearAccClamp=<3000 3000 3000>
    gravityOverride {z=-980}
    physicsBodyDefinitions [{
      boundBone {boneName=dyn_cloak_bk_l_01}
      boxExtents=<10 10 10>
      constraintSetup {angularLimitsMax=<0 5 0>, angularLimitsMin=<0 -5 -30>, angularTarget=<1 0 0>}
      localJointOffset=<2 0 0>
     }]
    simSpaceSettings {maxAngularAcceleration=45, maxAngularVelocity=8, simSpaceAngularAlpha=0.25}
   }
   Output Pose(Result=Component To Local(ComponentPose=@n12))
  ]}
 EventGraph {nodes [
   Event BlueprintUpdateAnimation()
   n1 = "Sin (Radians)"(A=float * float(A=12, B=GetGameTimeInSeconds()))
   > Set CloakRot1(CloakRot1:Rotator=MakeRotator(Pitch=float * float(A=@n1, B=10))) -> (Output_Get:Rotator)
  ]}
]
variables [
 CloakRot1 {type=Rotator}
 PoseAsset {type=PoseAsset, default=/CosmeticCompanions/Automation_Template/companion_pose_asset_template}
]
